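# Editor utility widget base for importing CSV data into Fortnite quest definition assets, with an empty canvas and stub hooks for CSV path, logging and quest editing.
# WriteToQuests checks out a cache DataTable, fills it from the CSV, then passes each row's matching quest to ManipulateQuest and saves it.
# Quest lookup falls back from package name to exact asset name to substring match.
# After the loop, the cache is saved and reverted in source control.
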EditorUtilityWidgetBlueprint /Game/NeverCook/EditorUtilities/QuestTools/QuestUtilities/SupportWidgets/EUW_QuestFromCSVImporter
graphs [
 CleanUpCache {nodes [
   CleanUpCache() {is=title K2Node_FunctionEntry}
   > SaveLoadedAsset(AssetToSave=Get importCache() -> (importCache:Data Table Object Reference), bOnlyIfIsDirty=false)
   > RevertFile!(execute, InFile:String=GetSystemPath!(Object:Object Reference=Get importCache()) -> (ReturnValue:String), bSilent:bool=false) -> (then, ReturnValue:bool)
  ]}
 EventGraph {nodes [
   GetCSVPath!() -> (CSV Path:String)
   LogToOutput!(execute, self, TextToAppend:String) -> (then)
   ManipulateQuest!(execute, self, QuestObject:Object Reference, DataTable:Data Table Object Reference, RowName:Name=None) -> (then)
  ]}
 GetCSVPath {nodes [
   GetCSVPath() {is=title K2Node_FunctionEntry}
   > Return Node!(execute, CSV Path:String)
  ]}
 LogToOutput {nodes [LogToOutput!() -> (then, TextToAppend:String) {is=title K2Node_FunctionEntry}]}
 ManipulateQuest {nodes [ManipulateQuest!() -> (then, QuestObject:Object Reference, DataTable:Data Table Object Reference, RowName:Name) {is=title K2Node_FunctionEntry}]}
 Prepare Cache {nodes [
   Prepare Cache() {is=title K2Node_FunctionEntry}
   > CheckoutLoadedAsset!(execute, AssetToCheckout:Object Reference=Get importCache()) -> (then, ReturnValue:bool)
  ]}
 WriteToQuests {nodes [
   WriteToQuests() {is=title K2Node_FunctionEntry}
   > Prepare Cache() {is=title K2Node_CallFunction}
   > n1 = Fill Data Table from CSV File!(
    execute
    DataTable:Data Table Object Reference=Get importCache()
    CSVFilePath:String=GetCSVPath() {is=title K2Node_CallFunction}
    ImportRowStruct:Script Struct Object Reference
   ) -> (then, ReturnValue:bool)
   > n2 = Branch(Condition=@n1)
   > n3 = GetDataTableRowNames(Table=Get importCache())
   > n4 = For Each Loop(Array:Array of Names=@n3) -> (Array Element:Name)
   CleanUpCache(execute=@n4.Completed) {is=title K2Node_CallFunction}
   n5 = Set currentRowName(execute=@n4.LoopBody, currentRowName:Name=@n4) -> (Output_Get:Name=None)
   > LogToOutput(TextToAppend=Append(A="Writing quest: ", B="To String (Name)"(InName=@n5))) {is=title K2Node_CallFunction}
   > n6 = findAssetByPackageOrName!(
    execute
    self
    assetPackageOrName:Name=Get currentRowName() -> (currentRowName:Name=None)
    ClassNamesOverride:Array of Names=Make Array("[0]":Name=FortQuestItemDefinition) -> (Array:Array of Names)
   ) -> (then, finalPackageName:Name=None, assetData:Asset Data Structure)
   > n7 = Set currentObject(currentObject:Object Reference=GetAsset(InAssetData=@n6.assetData)) -> (Output_Get:Object Reference)
   > ManipulateQuest(QuestObject=@n7, DataTable=Get importCache(), RowName=Get currentRowName() -> (currentRowName=None)) {is=title K2Node_CallFunction}
   > SaveLoadedAsset(AssetToSave=Get currentObject() -> (currentObject:Object Reference), bOnlyIfIsDirty=true)
   LogToOutput(execute=@n2.else, TextToAppend=Failed To Import CSV file. Aborting.)
  ]}
 findAssetByPackageOrName {nodes [
   n1 = findAssetByPackageOrName!() -> (then, assetPackageOrName:Name, ClassNamesOverride:Array of Names) {is=title K2Node_FunctionEntry}
   > Set l_name(l_name:Name=@n1) -> (Output_Get:Name=None)
   > Set l_classOverrides(l_classOverrides:Array of Names=@n1.ClassNamesOverride) -> (Output_Get:Array of Names)
   n2 = Get l_classOverrides() -> (l_classOverrides:Array of Names)
   > n3 = Branch(Condition="integer < integer"!(A:int=Length(TargetArray:Array of Names=@n2), B:int=1) -> (ReturnValue:bool) {is=title K2Node_CallFunction})
   n4 = Sequence!(execute=@n3.else) -> (then_0, then_1, then_2:Exec, then_3:Exec)
   n5 = GetAssetsByPackageName!(
    execute=@n4.then_1
    self:Asset Registry Interface=GetAssetRegistry()
    PackageName:Name=Get l_name() -> (l_name:Name=None)
    bIncludeOnlyOnDiskAssets:bool=false
    bSkipARFilteredAssets:bool=true
   ) -> (then, OutAssetData:Array of Asset Data Structures, ReturnValue:bool)
   > Branch(Condition="integer > integer"(A=Length(TargetArray:Array of Asset Data Structures=@n5), B=0) {is=title K2Node_CallFunction})
   > Return Node!(
    execute
    finalPackageName:Name=Get l_name() -> (l_name=None)
    assetData:Asset Data Structure="Get (a copy)"(Array:Array of Asset Data Structures=@n5) -> (Output:Asset Data Structure)
   )
   n6 = Make Array!("[0]_PackageName":String=/Script/FortniteGame, "[0]_AssetName":String=FortQuestItemDefinition) -> (Array:Array of Top Level Asset Path Structures)
   n7 = GetAssets!(
    execute=@n4.then_2
    self:Asset Registry Interface=GetAssetRegistry()
    Filter:"ARFilter Structure (by ref)"=MakeARFilter!(
     PackageNames:Array of Names
     PackagePaths:Array of Names
     SoftObjectPaths:Array of Soft Object Path Structures
     ClassPaths:Array of Top Level Asset Path Structures=@n6
     RecursiveClassPathsExclusionSet:Set of Top Level Asset Path Structures
     ClassNames:Array of Names
     RecursiveClassesExclusionSet:Set of Names
     bRecursivePaths:bool=false
     bRecursiveClasses:bool=true
     bIncludeOnlyOnDiskAssets:bool=false
    ) -> (ReturnValue:ARFilter Structure)
    bSkipARFilteredAssets:bool=true
   ) -> (then, OutAssetData:Array of Asset Data Structures, ReturnValue:bool)
   > n8 = For Each Loop with Break!(Exec:Exec, Array:Array of Asset Data Structures=@n7, Break:Exec) -> (LoopBody:Exec, Array Element:Asset Data Structure, Array Index:int, Completed:Exec)
   n9 = Break Asset Data(AssetData=@n8)
   n10 = Get l_name() -> (l_name=None)
   > Branch(Condition="Equal Exactly (String)"(A="To String (Name)"(InName=@n9.AssetName), B="To String (Name)"(InName=@n10)))
   > Return Node(finalPackageName=@n9, assetData=@n8)
   n11 = GetAssets(
    execute=@n4.then_3
    self=GetAssetRegistry()
    Filter=MakeARFilter(ClassPaths=@n6, bRecursivePaths=false, bRecursiveClasses=true, bIncludeOnlyOnDiskAssets=false)
    bSkipARFilteredAssets=true
   )
   > n12 = For Each Loop with Break(Array=@n11)
   n13 = Get l_name() -> (l_name=None)
   n14 = Break Asset Data(AssetData=@n12)
   > Branch(Condition=Contains(SearchIn="To String (Name)"(InName=@n14.AssetName), Substring="To String (Name)"(InName=@n13)))
   > Return Node(finalPackageName=@n14, assetData=@n12)
   Make Array!("[0]":Name=Blueprint, "[1]":Name=StaticMesh, "[2]":Name=SkeletalMesh) -> (Array:Array of Names)
  ]}
]
variables [importCache {type=DataTable, default=null}]
widgets {
 info {rootWidgetClass=&CanvasPanel, widgetCount=1}
 widgets [CanvasPanel_0 {widgetClassPath=&CanvasPanel, widget=CanvasPanel_0}]
}
