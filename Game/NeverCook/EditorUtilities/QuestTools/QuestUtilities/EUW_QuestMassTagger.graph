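# Editor utility widget that bulk-adds gameplay tags to Fortnite quest definitions from a picked CSV.
# It extends EUW_QuestFromCSVImporter: the button calls WriteToQuests, and overrides supply the CSV path, per-row ManipulateQuest and on-screen logging.
# Each row appends source, target and context tags to an objective's inline stat by container name, without transactions.
# The row's GameplayTag is appended to the returned owned-tags container, with no visible write-back.
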EditorUtilityWidgetBlueprint /Game/NeverCook/EditorUtilities/QuestTools/QuestUtilities/EUW_QuestMassTagger
graphs [
 EventGraph {nodes [
   "On Clicked (TA_Generic01_Button)"!() -> (OutputDelegate, then, widget:User Widget Object Reference, event:Pointer Event Structure) {is=title K2Node_ComponentBoundEvent}
   > WriteToQuests(self:EUW Quest From CSVImporter Object Reference)
  ]}
 GetCSVPath {nodes [
   GetCSVPath() {is=title K2Node_FunctionEntry}
   n1 = Get FilePicker_Generic_EUW() -> (FilePicker_Generic_EUW:File Picker Generic EUW Object Reference)
   > Return Node!(
    execute
    CSV Path:String="Get (a copy)"(Array=Get CurrentPaths!(self:File Picker Base EUW Object Reference=@n1) -> (CurrentPaths:Array of Strings))
   )
  ]}
 LogToOutput {nodes [
   n1 = LogToOutput!() -> (then, TextToAppend:String) {is=title K2Node_FunctionEntry}
   n2 = "To String (Text)"(InText="GetText (Text)"!(self:Text Object Reference=Get Output() -> (Output:Common Text Object Reference)) -> (ReturnValue:Text))
   > "SetText (Text)"(self=Get Output(), InText="To Text (String)"(InString=Append!(A=@n2, B="\r\n", C:String=@n1) -> (ReturnValue)))
  ]}
 ManipulateQuest {nodes [
   n1 = ManipulateQuest!() -> (then, QuestObject:Object Reference, DataTable:Data Table Object Reference, RowName:Name=None) {is=title K2Node_FunctionEntry}
   > Parent: ManipulateQuest!(execute, QuestObject:Object Reference=@n1, DataTable:Data Table Object Reference=@n1.DataTable, RowName:Name=@n1.RowName) -> (then) {is=title K2Node_CallParentFunction}
   > n2 = Get Data Table Row!(
    execute:Exec
    DataTable:Data Table Object Reference=Get DataTable() -> (DataTable:Data Table Object Reference)
    RowName:Name=Get RowName() -> (RowName:Name=None)
   ) -> (then:Exec, RowNotFound:Exec, ReturnValue:Quest Mass Tagger Row Structure)
   > n3 = Cast To FortQuestItemDefinition!(execute, Object=Get QuestObject() -> (QuestObject:Object Reference)) -> (then, CastFailed, AsFort Quest Item Definition:Fort Quest Item Definition Object Reference) {is=title K2Node_DynamicCast}
   > n4 = Set FortQuestItem(FortQuestItem:Fort Quest Item Definition Object Reference=@n3) -> (Output_Get:Fort Quest Item Definition Object Reference)
   n5 = Break Quest Mass Tagger Row!(QuestMassTaggerRow:"Quest Mass Tagger Row Structure (by ref)"=@n2) -> (
    ObjectiveNumber_6_#1:int
    InlineStatIndex_20_#2:int
    SourceTag_24_#3:String
    TargetTag_11_#4:String
    ContextTag_13_#5:String
    GameplayTag_15_#6:String
   ) {is=title K2Node_BreakStruct}
   > WriteGameplayTagFromString!(execute, self, TagName:String=@n5.GameplayTag_15_#6, Quest Item:Object Reference=@n4) -> (then)
   > WriteInlineStatTag!(
    execute
    self
    Quest:Fort Quest Item Definition Object Reference=Get FortQuestItem() -> (FortQuestItem:Fort Quest Item Definition Object Reference)
    ObjectiveIdx:int=@n5
    InlineStatIdx:int=@n5.InlineStatIndex_20_#2
    TagContainerPropertyName:Name=SourceTags
    TagName:String=@n5.SourceTag_24_#3
   ) -> (then)
   > WriteInlineStatTag(
    Quest=Get FortQuestItem()
    ObjectiveIdx=@n5
    InlineStatIdx=@n5.InlineStatIndex_20_#2
    TagContainerPropertyName=TargetTags
    TagName=@n5.TargetTag_11_#4
   )
   > WriteInlineStatTag(
    Quest=Get FortQuestItem()
    ObjectiveIdx=@n5
    InlineStatIdx=@n5.InlineStatIndex_20_#2
    TagContainerPropertyName=ContextTags
    TagName=@n5.ContextTag_13_#5
   )
   n6 = Get RowName() -> (RowName=None)
   LogToOutput!(
    execute=@n2.RowNotFound
    self:EUW Quest From CSVImporter Object Reference
    TextToAppend:String=Append!(A=Couldn't find the trable row:, B="To String (Name)"(InName=@n6)) -> (ReturnValue)
   ) -> (then) {is=title K2Node_CallFunction}
  ]}
 WriteGameplayTagFromString {nodes [
   n1 = WriteGameplayTagFromString!() -> (then, TagName:String, Quest Item:Object Reference) {is=title K2Node_FunctionEntry}
   > n2 = RequestGameplayTag!(execute, TagName:String=@n1, bErrorIfNotFound:bool=false) -> (then, OutTag:Gameplay Tag Structure, ReturnValue:bool)
   > n3 = Get Owned Gameplay Tags!(execute:Exec, self:Object Reference=Get Quest Item() -> (Quest Item:Object Reference)) -> (then:Exec, ReturnValue:Gameplay Tag Container Structure) {is=title K2Node_Message}
   > AppendGameplayTagContainers!(
    execute
    InOutTagContainer:"Gameplay Tag Container Structure (by ref)"=@n3
    InTagContainer:"Gameplay Tag Container Structure (by ref)"=MakeGameplayTagContainerFromTag!(SingleTag:Gameplay Tag Structure=@n2) -> (ReturnValue:Gameplay Tag Container Structure)
   ) -> (then)
  ]}
 WriteInlineStatTag {nodes [
   n1 = WriteInlineStatTag!() -> (then, Quest:Fort Quest Item Definition Object Reference, ObjectiveIdx:int, InlineStatIdx:int, TagContainerPropertyName:Name, TagName:String) {is=title K2Node_FunctionEntry}
   > LogToOutput(TextToAppend=Append(A="Adding Tag: ", B=@n1.TagName))
   > n2 = RequestGameplayTag(TagName=@n1.TagName, bErrorIfNotFound=false)
   > n3 = SetInlineStatGameplayTagContainerByName!(
    execute
    Quest:Fort Quest Item Definition Object Reference=@n1
    ObjectiveIdx:int=@n1.ObjectiveIdx
    InlineStatIdx:int=@n1.InlineStatIdx
    TagContainerPropertyName:Name=@n1.TagContainerPropertyName
    NewTags:Gameplay Tag Container Structure=MakeGameplayTagContainerFromTag(SingleTag=@n2)
    UpdateType:EFortGameplayTagContainerUpdateType Enum=Append
    bCreateTransaction:bool=false
   ) -> (then, ReturnValue:bool)
   > n4 = Branch(Condition=@n3)
   LogToOutput(execute=@n4.else, TextToAppend=Failed to set tag container.)
  ]}
]
variables []
widgets {
 info {
  parentClass=/Game/NeverCook/EditorUtilities/QuestTools/QuestUtilities/SupportWidgets/EUW_QuestFromCSVImporter.EUW_QuestFromCSVImporter_C
  rootWidgetClass=&VerticalBox
  widgetCount=10
 }
 widgets [
  VerticalBox_0 {widgetClassPath=&VerticalBox, widget=VerticalBox_0}
  CommonTextBlock_36 {
   widgetClassPath=&CommonTextBlock
   parent=VerticalBox_0
   widget=CommonTextBlock_36
   properties {
    autoWrapText=true
    style=/FortUILibrary/Styles/Typography/Foundation/TextStyle-Base-XS.TextStyle-Base-XS_C
    text=1. Make a CSV with this format:
    font {
     bForceMonospaced=false
     bMaterialIsStencil=false
     fontMaterial=null
     fontObject=&NotoSans
     letterSpacing=0
     monospacedWidth=1
     size=16
     skewAmount=0
     typefaceFontName=NotoSans-Regular
     outlineSettings {
      bApplyOutlineToDropShadows=false
      bMiteredCorners=false
      bSeparateFillAlpha=false
      outlineMaterial=null
      outlineSize=0
      outlineColor=<rgba 0 0 0 1>
     }
    }
    shadowOffset=<0 0>
   }
   slot {class=VerticalBoxSlot, properties {padding=<ltrb 5 5 5 5>}}
  }
  ScaleBox_128 {widgetClassPath=&ScaleBox, parent=VerticalBox_0, widget=ScaleBox_128, slot {class=VerticalBoxSlot}}
  CSVExampleImage {
   widgetClassPath=&Image
   bIsVariable=true
   parent=ScaleBox_128
   widget=CSVExampleImage
   properties {brush {
     bIsDynamicallyLoaded=false
     drawAs=Image
     imageType=NoImage
     mirroring=NoMirror
     resourceName=None
     resourceObject=/Game/NeverCook/EditorUtilities/QuestTools/QuestUtilities/Assets/MassTaggerTutorialGraphic
     tiling=NoTile
     imageSize=<1236 122>
     margin=<ltrb 0 0 0 0>
     outlineSettings {
      bUseBrushTransparency=false
      roundingType=HalfHeightRadius
      width=0
      color {colorUseRule=UseColor_Specified, specifiedColor=<rgba 0 0 0 0>}
      cornerRadii=<0 0 0 0>
     }
     tintColor {colorUseRule=UseColor_Specified, specifiedColor=<rgba 1 1 1 1>}
     uVRegion {bIsValid=false, max=<0 0>, min=<0 0>}
    }}
   slot {class=ScaleBoxSlot, properties {horizontalAlignment=HAlign_Left, verticalAlignment=VAlign_Top}}
  }
  CommonTextBlock_1 {
   widgetClassPath=&CommonTextBlock
   parent=VerticalBox_0
   widget=CommonTextBlock_1
   properties {
    autoWrapText=true
    style=&TextStyle-Base-XS_C
    text=2. Pick your csv file here:
    font {
     bForceMonospaced=false
     bMaterialIsStencil=false
     fontMaterial=null
     fontObject=&NotoSans
     letterSpacing=0
     monospacedWidth=1
     size=16
     skewAmount=0
     typefaceFontName=NotoSans-Regular
     outlineSettings {
      bApplyOutlineToDropShadows=false
      bMiteredCorners=false
      bSeparateFillAlpha=false
      outlineMaterial=null
      outlineSize=0
      outlineColor=<rgba 0 0 0 1>
     }
    }
    shadowOffset=<0 0>
   }
   slot {class=VerticalBoxSlot, properties {padding=<ltrb 5 5 5 5>}}
  }
  FilePicker_Generic_EUW {
   widgetClassPath=/Game/Blueprints/Blutilities/ForTechArt/Widgets/FilePicker/FilePicker_Generic_EUW
   bIsVariable=true
   parent=VerticalBox_0
   widget=FilePicker_Generic_EUW
   slot {class=VerticalBoxSlot, properties {padding=<ltrb 5 5 5 5>}}
  }
  CommonTextBlock {
   widgetClassPath=&CommonTextBlock
   parent=VerticalBox_0
   widget=CommonTextBlock
   properties {
    autoWrapText=true
    style=&TextStyle-Base-XS_C
    text=3. Press this button:
    font {
     bForceMonospaced=false
     bMaterialIsStencil=false
     fontMaterial=null
     fontObject=&NotoSans
     letterSpacing=0
     monospacedWidth=1
     size=16
     skewAmount=0
     typefaceFontName=NotoSans-Regular
     outlineSettings {
      bApplyOutlineToDropShadows=false
      bMiteredCorners=false
      bSeparateFillAlpha=false
      outlineMaterial=null
      outlineSize=0
      outlineColor=<rgba 0 0 0 1>
     }
    }
    shadowOffset=<0 0>
   }
   slot {class=VerticalBoxSlot, properties {padding=<ltrb 5 5 5 5>}}
  }
  TA_Generic01_Button {
   widgetClassPath=&TA_Generic01_Button
   parent=VerticalBox_0
   widget=TA_Generic01_Button
   properties {label=Import the tags into quests}
   slot {class=VerticalBoxSlot}
  }
  ScrollBox_63 {
   widgetClassPath=&ScrollBox
   parent=VerticalBox_0
   widget=ScrollBox_63
   properties {
    widgetBarStyle {
     thickness=8
     draggedThumbImage {
      bIsDynamicallyLoaded=false
      drawAs=RoundedBox
      imageType=NoImage
      mirroring=NoMirror
      resourceName=None
      resourceObject=null
      tiling=NoTile
      imageSize=<0 0>
      margin=<ltrb 0 0 0 0>
      outlineSettings {
       bUseBrushTransparency=false
       roundingType=FixedRadius
       width=0
       color {colorUseRule=UseColor_Specified, specifiedColor=<rgba 0 0 0 0>}
       cornerRadii=<4 4 4 4>
      }
      tintColor {colorUseRule=UseColor_Specified, specifiedColor=<rgba 0.2159 0.2159 0.2159 1>}
      uVRegion {bIsValid=false, max=<0 0>, min=<0 0>}
     }
     horizontalBackgroundImage {
      bIsDynamicallyLoaded=false
      drawAs=NoDrawType
      imageType=NoImage
      mirroring=NoMirror
      resourceName=None
      resourceObject=null
      tiling=NoTile
      imageSize=<0 0>
      margin=<ltrb 0 0 0 0>
      outlineSettings {
       bUseBrushTransparency=false
       roundingType=HalfHeightRadius
       width=0
       color {colorUseRule=UseColor_Specified, specifiedColor=<rgba 0 0 0 0>}
       cornerRadii=<0 0 0 0>
      }
      tintColor {colorUseRule=UseColor_Specified, specifiedColor=<rgba 1 1 1 1>}
      uVRegion {bIsValid=false, max=<0 0>, min=<0 0>}
     }
     horizontalBottomSlotImage {
      bIsDynamicallyLoaded=false
      drawAs=NoDrawType
      imageType=NoImage
      mirroring=NoMirror
      resourceName=None
      resourceObject=null
      tiling=NoTile
      imageSize=<0 0>
      margin=<ltrb 0 0 0 0>
      outlineSettings {
       bUseBrushTransparency=false
       roundingType=HalfHeightRadius
       width=0
       color {colorUseRule=UseColor_Specified, specifiedColor=<rgba 0 0 0 0>}
       cornerRadii=<0 0 0 0>
      }
      tintColor {colorUseRule=UseColor_Specified, specifiedColor=<rgba 1 1 1 1>}
      uVRegion {bIsValid=false, max=<0 0>, min=<0 0>}
     }
     horizontalTopSlotImage {
      bIsDynamicallyLoaded=false
      drawAs=NoDrawType
      imageType=NoImage
      mirroring=NoMirror
      resourceName=None
      resourceObject=null
      tiling=NoTile
      imageSize=<0 0>
      margin=<ltrb 0 0 0 0>
      outlineSettings {
       bUseBrushTransparency=false
       roundingType=HalfHeightRadius
       width=0
       color {colorUseRule=UseColor_Specified, specifiedColor=<rgba 0 0 0 0>}
       cornerRadii=<0 0 0 0>
      }
      tintColor {colorUseRule=UseColor_Specified, specifiedColor=<rgba 1 1 1 1>}
      uVRegion {bIsValid=false, max=<0 0>, min=<0 0>}
     }
     hoveredThumbImage {
      bIsDynamicallyLoaded=false
      drawAs=RoundedBox
      imageType=NoImage
      mirroring=NoMirror
      resourceName=None
      resourceObject=null
      tiling=NoTile
      imageSize=<0 0>
      margin=<ltrb 0 0 0 0>
      outlineSettings {
       bUseBrushTransparency=false
       roundingType=FixedRadius
       width=0
       color {colorUseRule=UseColor_Specified, specifiedColor=<rgba 0 0 0 0>}
       cornerRadii=<4 4 4 4>
      }
      tintColor {colorUseRule=UseColor_Specified, specifiedColor=<rgba 0.2159 0.2159 0.2159 1>}
      uVRegion {bIsValid=false, max=<0 0>, min=<0 0>}
     }
     normalThumbImage {
      bIsDynamicallyLoaded=false
      drawAs=RoundedBox
      imageType=NoImage
      mirroring=NoMirror
      resourceName=None
      resourceObject=null
      tiling=NoTile
      imageSize=<0 0>
      margin=<ltrb 0 0 0 0>
      outlineSettings {
       bUseBrushTransparency=false
       roundingType=FixedRadius
       width=0
       color {colorUseRule=UseColor_Specified, specifiedColor=<rgba 0 0 0 0>}
       cornerRadii=<4 4 4 4>
      }
      tintColor {colorUseRule=UseColor_Specified, specifiedColor=<rgba 0.0953 0.0953 0.0953 1>}
      uVRegion {bIsValid=false, max=<0 0>, min=<0 0>}
     }
     verticalBackgroundImage {
      bIsDynamicallyLoaded=false
      drawAs=NoDrawType
      imageType=NoImage
      mirroring=NoMirror
      resourceName=None
      resourceObject=null
      tiling=NoTile
      imageSize=<0 0>
      margin=<ltrb 0 0 0 0>
      outlineSettings {
       bUseBrushTransparency=false
       roundingType=HalfHeightRadius
       width=0
       color {colorUseRule=UseColor_Specified, specifiedColor=<rgba 0 0 0 0>}
       cornerRadii=<0 0 0 0>
      }
      tintColor {colorUseRule=UseColor_Specified, specifiedColor=<rgba 1 1 1 1>}
      uVRegion {bIsValid=false, max=<0 0>, min=<0 0>}
     }
     verticalBottomSlotImage {
      bIsDynamicallyLoaded=false
      drawAs=NoDrawType
      imageType=NoImage
      mirroring=NoMirror
      resourceName=None
      resourceObject=null
      tiling=NoTile
      imageSize=<0 0>
      margin=<ltrb 0 0 0 0>
      outlineSettings {
       bUseBrushTransparency=false
       roundingType=HalfHeightRadius
       width=0
       color {colorUseRule=UseColor_Specified, specifiedColor=<rgba 0 0 0 0>}
       cornerRadii=<0 0 0 0>
      }
      tintColor {colorUseRule=UseColor_Specified, specifiedColor=<rgba 1 1 1 1>}
      uVRegion {bIsValid=false, max=<0 0>, min=<0 0>}
     }
     verticalTopSlotImage {
      bIsDynamicallyLoaded=false
      drawAs=NoDrawType
      imageType=NoImage
      mirroring=NoMirror
      resourceName=None
      resourceObject=null
      tiling=NoTile
      imageSize=<0 0>
      margin=<ltrb 0 0 0 0>
      outlineSettings {
       bUseBrushTransparency=false
       roundingType=HalfHeightRadius
       width=0
       color {colorUseRule=UseColor_Specified, specifiedColor=<rgba 0 0 0 0>}
       cornerRadii=<0 0 0 0>
      }
      tintColor {colorUseRule=UseColor_Specified, specifiedColor=<rgba 1 1 1 1>}
      uVRegion {bIsValid=false, max=<0 0>, min=<0 0>}
     }
    }
    widgetStyle {
     barThickness=8
     bottomShadowBrush {
      bIsDynamicallyLoaded=false
      drawAs=Box
      imageType=FullColor
      mirroring=NoMirror
      resourceName=../../../Engine/Content/Slate/Common/ScrollBoxShadowBottom.png
      resourceObject=null
      tiling=NoTile
      imageSize=<16 8>
      margin=<ltrb 0.5 0 0.5 1>
      outlineSettings {
       bUseBrushTransparency=false
       roundingType=HalfHeightRadius
       width=0
       color {colorUseRule=UseColor_Specified, specifiedColor=<rgba 0 0 0 0>}
       cornerRadii=<0 0 0 0>
      }
      tintColor {colorUseRule=UseColor_Specified, specifiedColor=<rgba 1 1 1 1>}
      uVRegion {bIsValid=false, max=<0 0>, min=<0 0>}
     }
     horizontalScrolledContentPadding=<ltrb 0 0 1 0>
     leftShadowBrush {
      bIsDynamicallyLoaded=false
      drawAs=Box
      imageType=FullColor
      mirroring=NoMirror
      resourceName=../../../Engine/Content/Slate/Common/ScrollBoxShadowLeft.png
      resourceObject=null
      tiling=NoTile
      imageSize=<8 16>
      margin=<ltrb 1 0.5 0 0.5>
      outlineSettings {
       bUseBrushTransparency=false
       roundingType=HalfHeightRadius
       width=0
       color {colorUseRule=UseColor_Specified, specifiedColor=<rgba 0 0 0 0>}
       cornerRadii=<0 0 0 0>
      }
      tintColor {colorUseRule=UseColor_Specified, specifiedColor=<rgba 1 1 1 1>}
      uVRegion {bIsValid=false, max=<0 0>, min=<0 0>}
     }
     rightShadowBrush {
      bIsDynamicallyLoaded=false
      drawAs=Box
      imageType=FullColor
      mirroring=NoMirror
      resourceName=../../../Engine/Content/Slate/Common/ScrollBoxShadowRight.png
      resourceObject=null
      tiling=NoTile
      imageSize=<8 16>
      margin=<ltrb 0 0.5 1 0.5>
      outlineSettings {
       bUseBrushTransparency=false
       roundingType=HalfHeightRadius
       width=0
       color {colorUseRule=UseColor_Specified, specifiedColor=<rgba 0 0 0 0>}
       cornerRadii=<0 0 0 0>
      }
      tintColor {colorUseRule=UseColor_Specified, specifiedColor=<rgba 1 1 1 1>}
      uVRegion {bIsValid=false, max=<0 0>, min=<0 0>}
     }
     topShadowBrush {
      bIsDynamicallyLoaded=false
      drawAs=Box
      imageType=FullColor
      mirroring=NoMirror
      resourceName=../../../Engine/Content/Slate/Common/ScrollBoxShadowTop.png
      resourceObject=null
      tiling=NoTile
      imageSize=<16 8>
      margin=<ltrb 0.5 1 0.5 0>
      outlineSettings {
       bUseBrushTransparency=false
       roundingType=HalfHeightRadius
       width=0
       color {colorUseRule=UseColor_Specified, specifiedColor=<rgba 0 0 0 0>}
       cornerRadii=<0 0 0 0>
      }
      tintColor {colorUseRule=UseColor_Specified, specifiedColor=<rgba 1 1 1 1>}
      uVRegion {bIsValid=false, max=<0 0>, min=<0 0>}
     }
     verticalScrolledContentPadding=<ltrb 0 0 0 1>
    }
   }
   slot {class=VerticalBoxSlot, properties {size {sizeRule=Fill, value=1}}}
  }
  Output {
   widgetClassPath=&CommonTextBlock
   bIsVariable=true
   parent=ScrollBox_63
   widget=Output
   properties {
    autoWrapText=true
    style=&TextStyle-Base-XS_C
    text=Output:
    font {
     bForceMonospaced=false
     bMaterialIsStencil=false
     fontMaterial=null
     fontObject=&NotoSans
     letterSpacing=0
     monospacedWidth=1
     size=16
     skewAmount=0
     typefaceFontName=NotoSans-Regular
     outlineSettings {
      bApplyOutlineToDropShadows=false
      bMiteredCorners=false
      bSeparateFillAlpha=false
      outlineMaterial=null
      outlineSize=0
      outlineColor=<rgba 0 0 0 1>
     }
    }
    shadowOffset=<0 0>
   }
   slot {class=ScrollBoxSlot, properties {padding=<ltrb 50 50 50 50>}}
  }
 ]
}
