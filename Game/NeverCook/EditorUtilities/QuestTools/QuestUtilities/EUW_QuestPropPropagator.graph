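# Editor utility widget that creates child quest prop actor Blueprints for a picked quest and parent prop actor class.
# Its Create Child Actors button passes both picks to the native FortQuestPropPropagator parent's MakePropActors.
# The SetPropActorData event runs SetUpChild, which writes quest, objective backend name and editor tags to the CDO, then saves.
# An older Blueprint creation loop, using asset-registry search and Python CDO lookup, remains unconnected in MakeChildActors.
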EditorUtilityWidgetBlueprint /Game/NeverCook/EditorUtilities/QuestTools/QuestUtilities/EUW_QuestPropPropagator
graphs [
 EventGraph {nodes [
   "On Clicked (CreateChildActorsButton)"!() -> (OutputDelegate, then, widget:User Widget Object Reference, event:Pointer Event Structure) {is=title K2Node_ComponentBoundEvent}
   > MakeChildActors()
   Event PreConstruct()
   > SetObject(
    self=[
     Get ParentPropPicker() -> (ParentPropPicker:Fort Quest Tool Property View Object Reference)
     Get QuestPicker() -> (QuestPicker:Fort Quest Tool Property View Object Reference)
    ]
    NewObject=Self-Reference()
   )
   n1 = Event SetPropActorData!() -> (OutputDelegate, then, BlueprintObject:Blueprint Object Reference, BlueprintObjectCDO:Object Reference, ObjectiveIndex:int=0) {is=title K2Node_Event}
   > SetUpChild!(
    execute
    self
    BlueprintObject:Object Reference=@n1.BlueprintObject
    BlueprintObjectCDO:Object Reference=@n1.BlueprintObjectCDO
    ObjectiveIndex:int=@n1.ObjectiveIndex
   ) -> (then)
  ]}
 GetClassDefaultObject {nodes [
   n1 = GetClassDefaultObject!() -> (then, asset_data:Asset Data Structure) {is=title K2Node_FunctionEntry}
   n2 = Break Asset Data(AssetData=@n1)
   n3 = Execute Python Script!(
    execute
    PythonScript="import unreal\r\nprint(asset_class)\r\nclass_ref = unreal.load_class(asset_class)\r\nclass_default_object = unreal.get_default_object(class_ref)\r\nprint(class_default_object)"
    asset_class:String=Append!(A="To String (Name)"(InName=@n2.PackagePath), B="To String (Name)"(InName=@n2), C:String=_C) -> (ReturnValue)
   ) -> (then, ReturnValue, ClassDefaultObject:Wildcard) {inputs [asset_class], outputs [ClassDefaultObject]}
   n4 = Execute Python Script!(
    execute=@n1.then
    PythonScript="import unreal\r\nprint(asset_data)\r\nasset_class = unreal.load_class(None, str(asset_data.object_path) + \"_C\")\r\nclass_default_object = unreal.get_default_object(asset_class)\r\nprint(class_default_object)"
    asset_data:Asset Data Structure=@n1
   ) -> (then, ReturnValue, ClassDefaultObject:Object Reference) {inputs [asset_data], outputs [ClassDefaultObject]}
   n5 = Set LocalCDO(execute=[@n3.then, @n4.then], LocalCDO:Object Reference=@n4.ClassDefaultObject) -> (Output_Get:Object Reference)
   > PrintString(InString=Append!(A="Searched found a CDO and found: ", B=GetDisplayName(Object=@n5)) -> (ReturnValue))
   > Return Node!(execute, Class Default Object:Object Reference=Get LocalCDO() -> (LocalCDO:Object Reference))
   MakeLiteralString(Value=BP_S20_BC_DiscoverMounds_C)
   Get ParentQuestPropActor() -> (ParentQuestPropActor:BP Athena Prop Quest Actor Parent Class Reference)
  ]}
 MakeChildActors {nodes [
   n1 = GetAssetTools()
   > Set AssetTools(AssetTools:Asset Tools Interface=@n1) -> (Output_Get:Asset Tools Interface)
   n2 = Get Class Display Name(Class=Get ParentQuestPropActor())
   > n3 = findAssetByPackageOrName!(execute, self, assetPackageOrName:Name=String To Name(InString=LeftChop(SourceString=@n2, Count=2))) -> (then, finalPackageName:Name=None, assetData:Asset Data Structure)
   > n4 = Set ParentCDO(ParentCDO:Object Reference=GetClassDefaultObject!(asset_data:Asset Data Structure=@n3.assetData) -> (Class Default Object:Object Reference) {is=title K2Node_CallFunction}) -> (Output_Get:Object Reference)
   > n5 = GetPathNameForLoadedAsset(LoadedAsset=@n4)
   > n6 = FindAssetData!(execute, AssetPath:String=@n5) -> (then, ReturnValue:Asset Data Structure)
   n7 = Break Asset Data!(AssetData=@n6) -> (PackagePath, AssetName, AssetClassPath)
   > Set ParentPropName(ParentPropName:String="To String (Name)"(InName=@n7.AssetName)) -> (Output_Get:String)
   > Set DestinationPath(DestinationPath:String=Append!(A="To String (Name)"(InName=@n7), B=/, C:String=ChildActors) -> (ReturnValue)) -> (Output_Get:String)
   > n8 = For Each Loop(Array:Array of Fort Mcp Quest Objective Info Structures=GetObjectives!(self:Fort Quest Item Definition Object Reference=Get Quest() -> (Quest:Fort Quest Item Definition Object Reference)) -> (ReturnValue:Array of Fort Mcp Quest Objective Info Structures)) -> (Array Element:Fort Mcp Quest Objective Info Structure)
   n9 = Append(
    A=Replace(SourceString=Get ParentPropName() -> (ParentPropName:String), From=Default_)
    B=_Objective
    C="To String (Integer)"(InInt=@n8.Array Index)
   )
   > n10 = CreateAsset(
    self=Get AssetTools() -> (AssetTools:Asset Tools Interface)
    AssetName=@n9
    PackagePath=Get DestinationPath() -> (DestinationPath:String)
    AssetClass=Get ParentQuestPropActor()
   )
   > SetUpChild(BlueprintObject=@n10, ObjectiveIndex=@n8.Array Index)
   MakeChildActors() {is=title K2Node_FunctionEntry}
   > MakePropActors!(
    execute
    self:Fort Quest Prop Propagator Object Reference
    InQuestPropParentClass:Object Class Reference=Get ParentQuestPropActor()
    QuestItemDef:Fort Quest Item Definition Object Reference=Get Quest()
   ) -> (then)
   String To Name(InString=@n9)
  ]}
 SetUpChild {nodes [
   n1 = SetUpChild!() -> (then, BlueprintObject:Object Reference, BlueprintObjectCDO:Object Reference, ObjectiveIndex:int) {is=title K2Node_FunctionEntry}
   > n2 = Cast To BP_Athena_PropQuestActor_Parent!(execute, Object=@n1.BlueprintObjectCDO) -> (then, CastFailed, AsBP Athena Prop Quest Actor Parent:BP Athena Prop Quest Actor Parent Object Reference) {is=title K2Node_DynamicCast}
   n3 = Get ConsolidatedQuestComponent!(self:Building Gameplay Actor Prop Quest Object Reference=@n2) -> (ConsolidatedQuestComponent:Consolidated Quest Component Object Reference)
   > SetQuestItemDefinition!(
    execute
    self:Consolidated Quest Component Object Reference=@n3
    InQuestItemDefinition:Fort Quest Item Definition Object Reference=Get Quest()
   ) -> (then)
   n4 = "Get (a ref)"(
    Array:Array of Fort Mcp Quest Objective Info Structures=GetObjectives(self=Get Quest())
    Dimension 1=Get ObjectiveIndex() -> (ObjectiveIndex:int=0)
   ) -> (Output:"Fort Mcp Quest Objective Info Structure (by ref)")
   > SetObjectiveBackendName!(
    execute
    self:Consolidated Quest Component Object Reference=@n3
    InObjectiveBackendName:"Name (by ref)"=Break Fort Mcp Quest Objective Info!(FortMcpQuestObjectiveInfo:"Fort Mcp Quest Objective Info Structure (by ref)"=@n4) -> (
     BackendName:Name
     ItemEvent:EFortQuestObjectiveItemEvent Enum
     bHidden:bool=~
     bRequirePrimaryMissionCompletion:bool=~
     ObjectiveStatHandle:Data Table Row Handle Structure
     AlternativeStatHandles:Array of Data Table Row Handle Structures
     ItemReference:Fort Item Definition Soft Object Reference
     Description:Text
     HudShortDescription:Text
     Count:int
    ) {is=title K2Node_BreakStruct}
   ) -> (then)
   > n5 = GetQuestObjectiveTagForQuestProp!(execute, Objective:"Fort Mcp Quest Objective Info Structure (by ref)"=@n4) -> (then, OutContainerTag:Gameplay Tag Container Structure, ReturnValue:bool)
   > Set QuestActorEditorOnlyInstanceTags!(execute, QuestActorEditorOnlyInstanceTags:Gameplay Tag Container Structure=@n5, self:Building Gameplay Actor Quest Object Reference=@n2) -> (then, Output_Get:Gameplay Tag Container Structure)
   > SaveLoadedAsset(AssetToSave=Get BlueprintObject() -> (BlueprintObject:Object Reference), bOnlyIfIsDirty=false)
   PrintString(execute=@n2.CastFailed, InString=CAST FAILED HELLLO, Duration=52)
  ]}
 findAssetByPackageOrName {nodes [
   n1 = findAssetByPackageOrName!() -> (then, assetPackageOrName:Name) {is=title K2Node_FunctionEntry}
   > Set l_name(l_name:Name=@n1) -> (Output_Get:Name=None)
   > n2 = Sequence!(execute) -> (then_0, then_1, then_2:Exec, then_3:Exec)
   n3 = GetAssetsByPackageName!(
    execute=@n2.then_1
    self:Asset Registry Interface=GetAssetRegistry()
    PackageName:Name=Get l_name() -> (l_name:Name=None)
    bIncludeOnlyOnDiskAssets:bool=false
    bSkipARFilteredAssets:bool=true
   ) -> (then, OutAssetData:Array of Asset Data Structures, ReturnValue:bool)
   > Branch(Condition="integer > integer"(A=Length(TargetArray:Array of Asset Data Structures=@n3), B=0) {is=title K2Node_CallFunction})
   > Return Node!(
    execute
    finalPackageName:Name=Get l_name() -> (l_name=None)
    assetData:Asset Data Structure="Get (a copy)"(Array:Array of Asset Data Structures=@n3) -> (Output:Asset Data Structure)
   )
   n4 = Make Array("[0]":Name=Blueprint) -> (Array:Array of Names)
   n5 = GetAssets!(
    execute=@n2.then_2
    self:Asset Registry Interface=GetAssetRegistry()
    Filter:"ARFilter Structure (by ref)"=MakeARFilter!(
     PackageNames:Array of Names
     PackagePaths:Array of Names
     SoftObjectPaths:Array of Soft Object Path Structures
     ClassPaths:Array of Top Level Asset Path Structures
     RecursiveClassPathsExclusionSet:Set of Top Level Asset Path Structures
     ClassNames:Array of Names=@n4
     RecursiveClassesExclusionSet:Set of Names
     bRecursivePaths:bool=false
     bRecursiveClasses:bool=true
     bIncludeOnlyOnDiskAssets:bool=false
    ) -> (ReturnValue:ARFilter Structure)
    bSkipARFilteredAssets:bool=true
   ) -> (then, OutAssetData:Array of Asset Data Structures, ReturnValue:bool)
   > n6 = For Each Loop with Break!(Exec:Exec, Array:Array of Asset Data Structures=@n5, Break:Exec) -> (LoopBody:Exec, Array Element:Asset Data Structure, Array Index:int, Completed:Exec)
   n7 = Break Asset Data!(AssetData=@n6) -> (PackageName:Name, PackagePath, AssetName, AssetClassPath)
   n8 = Get l_name() -> (l_name=None)
   > Branch(Condition="Equal Exactly (String)"(A="To String (Name)"(InName=@n7.AssetName), B="To String (Name)"(InName=@n8)))
   > Return Node(finalPackageName=@n7, assetData=@n6)
   n9 = GetAssets(
    execute=@n2.then_3
    self=GetAssetRegistry()
    Filter=MakeARFilter(ClassNames=@n4, bRecursivePaths=false, bRecursiveClasses=true, bIncludeOnlyOnDiskAssets=false)
    bSkipARFilteredAssets=true
   )
   > n10 = For Each Loop with Break(Array=@n9)
   n11 = Get l_name() -> (l_name=None)
   n12 = Break Asset Data(AssetData=@n10)
   > Branch(Condition=Contains(SearchIn="To String (Name)"(InName=@n12.AssetName), Substring="To String (Name)"(InName=@n11)))
   > Return Node(finalPackageName=@n12, assetData=@n10)
   Make Array!("[0]"=Blueprint, "[1]":Name=StaticMesh, "[2]":Name=SkeletalMesh) -> (Array)
  ]}
]
variables [
 ParentQuestPropActor {type=Class, default=null}
 Quest {type=FortQuestItemDefinition, default=null}
]
widgets {
 info {parentClass=/Script/FortniteEditor.FortQuestPropPropagator, rootWidgetClass=&VerticalBox, widgetCount=10}
 widgets [
  VerticalBox_0 {widgetClassPath=&VerticalBox, widget=VerticalBox_0}
  ScrollBox_0 {
   widgetClassPath=&ScrollBox
   parent=VerticalBox_0
   widget=ScrollBox_0
   properties {
    widgetBarStyle {
     thickness=8
     draggedThumbImage {
      bIsDynamicallyLoaded=false
      drawAs=RoundedBox
      imageType=NoImage
      mirroring=NoMirror
      resourceName=None
      resourceObject=null
      tiling=NoTile
      imageSize=<0 0>
      margin=<ltrb 0 0 0 0>
      outlineSettings {
       bUseBrushTransparency=false
       roundingType=FixedRadius
       width=0
       color {colorUseRule=UseColor_Specified, specifiedColor=<rgba 0 0 0 0>}
       cornerRadii=<4 4 4 4>
      }
      tintColor {colorUseRule=UseColor_Specified, specifiedColor=<rgba 0.2159 0.2159 0.2159 1>}
      uVRegion {bIsValid=false, max=<0 0>, min=<0 0>}
     }
     horizontalBackgroundImage {
      bIsDynamicallyLoaded=false
      drawAs=NoDrawType
      imageType=NoImage
      mirroring=NoMirror
      resourceName=None
      resourceObject=null
      tiling=NoTile
      imageSize=<0 0>
      margin=<ltrb 0 0 0 0>
      outlineSettings {
       bUseBrushTransparency=false
       roundingType=HalfHeightRadius
       width=0
       color {colorUseRule=UseColor_Specified, specifiedColor=<rgba 0 0 0 0>}
       cornerRadii=<0 0 0 0>
      }
      tintColor {colorUseRule=UseColor_Specified, specifiedColor=<rgba 1 1 1 1>}
      uVRegion {bIsValid=false, max=<0 0>, min=<0 0>}
     }
     horizontalBottomSlotImage {
      bIsDynamicallyLoaded=false
      drawAs=NoDrawType
      imageType=NoImage
      mirroring=NoMirror
      resourceName=None
      resourceObject=null
      tiling=NoTile
      imageSize=<0 0>
      margin=<ltrb 0 0 0 0>
      outlineSettings {
       bUseBrushTransparency=false
       roundingType=HalfHeightRadius
       width=0
       color {colorUseRule=UseColor_Specified, specifiedColor=<rgba 0 0 0 0>}
       cornerRadii=<0 0 0 0>
      }
      tintColor {colorUseRule=UseColor_Specified, specifiedColor=<rgba 1 1 1 1>}
      uVRegion {bIsValid=false, max=<0 0>, min=<0 0>}
     }
     horizontalTopSlotImage {
      bIsDynamicallyLoaded=false
      drawAs=NoDrawType
      imageType=NoImage
      mirroring=NoMirror
      resourceName=None
      resourceObject=null
      tiling=NoTile
      imageSize=<0 0>
      margin=<ltrb 0 0 0 0>
      outlineSettings {
       bUseBrushTransparency=false
       roundingType=HalfHeightRadius
       width=0
       color {colorUseRule=UseColor_Specified, specifiedColor=<rgba 0 0 0 0>}
       cornerRadii=<0 0 0 0>
      }
      tintColor {colorUseRule=UseColor_Specified, specifiedColor=<rgba 1 1 1 1>}
      uVRegion {bIsValid=false, max=<0 0>, min=<0 0>}
     }
     hoveredThumbImage {
      bIsDynamicallyLoaded=false
      drawAs=RoundedBox
      imageType=NoImage
      mirroring=NoMirror
      resourceName=None
      resourceObject=null
      tiling=NoTile
      imageSize=<0 0>
      margin=<ltrb 0 0 0 0>
      outlineSettings {
       bUseBrushTransparency=false
       roundingType=FixedRadius
       width=0
       color {colorUseRule=UseColor_Specified, specifiedColor=<rgba 0 0 0 0>}
       cornerRadii=<4 4 4 4>
      }
      tintColor {colorUseRule=UseColor_Specified, specifiedColor=<rgba 0.2159 0.2159 0.2159 1>}
      uVRegion {bIsValid=false, max=<0 0>, min=<0 0>}
     }
     normalThumbImage {
      bIsDynamicallyLoaded=false
      drawAs=RoundedBox
      imageType=NoImage
      mirroring=NoMirror
      resourceName=None
      resourceObject=null
      tiling=NoTile
      imageSize=<0 0>
      margin=<ltrb 0 0 0 0>
      outlineSettings {
       bUseBrushTransparency=false
       roundingType=FixedRadius
       width=0
       color {colorUseRule=UseColor_Specified, specifiedColor=<rgba 0 0 0 0>}
       cornerRadii=<4 4 4 4>
      }
      tintColor {colorUseRule=UseColor_Specified, specifiedColor=<rgba 0.0953 0.0953 0.0953 1>}
      uVRegion {bIsValid=false, max=<0 0>, min=<0 0>}
     }
     verticalBackgroundImage {
      bIsDynamicallyLoaded=false
      drawAs=NoDrawType
      imageType=NoImage
      mirroring=NoMirror
      resourceName=None
      resourceObject=null
      tiling=NoTile
      imageSize=<0 0>
      margin=<ltrb 0 0 0 0>
      outlineSettings {
       bUseBrushTransparency=false
       roundingType=HalfHeightRadius
       width=0
       color {colorUseRule=UseColor_Specified, specifiedColor=<rgba 0 0 0 0>}
       cornerRadii=<0 0 0 0>
      }
      tintColor {colorUseRule=UseColor_Specified, specifiedColor=<rgba 1 1 1 1>}
      uVRegion {bIsValid=false, max=<0 0>, min=<0 0>}
     }
     verticalBottomSlotImage {
      bIsDynamicallyLoaded=false
      drawAs=NoDrawType
      imageType=NoImage
      mirroring=NoMirror
      resourceName=None
      resourceObject=null
      tiling=NoTile
      imageSize=<0 0>
      margin=<ltrb 0 0 0 0>
      outlineSettings {
       bUseBrushTransparency=false
       roundingType=HalfHeightRadius
       width=0
       color {colorUseRule=UseColor_Specified, specifiedColor=<rgba 0 0 0 0>}
       cornerRadii=<0 0 0 0>
      }
      tintColor {colorUseRule=UseColor_Specified, specifiedColor=<rgba 1 1 1 1>}
      uVRegion {bIsValid=false, max=<0 0>, min=<0 0>}
     }
     verticalTopSlotImage {
      bIsDynamicallyLoaded=false
      drawAs=NoDrawType
      imageType=NoImage
      mirroring=NoMirror
      resourceName=None
      resourceObject=null
      tiling=NoTile
      imageSize=<0 0>
      margin=<ltrb 0 0 0 0>
      outlineSettings {
       bUseBrushTransparency=false
       roundingType=HalfHeightRadius
       width=0
       color {colorUseRule=UseColor_Specified, specifiedColor=<rgba 0 0 0 0>}
       cornerRadii=<0 0 0 0>
      }
      tintColor {colorUseRule=UseColor_Specified, specifiedColor=<rgba 1 1 1 1>}
      uVRegion {bIsValid=false, max=<0 0>, min=<0 0>}
     }
    }
    widgetStyle {
     barThickness=8
     bottomShadowBrush {
      bIsDynamicallyLoaded=false
      drawAs=Box
      imageType=FullColor
      mirroring=NoMirror
      resourceName=../../../Engine/Content/Slate/Common/ScrollBoxShadowBottom.png
      resourceObject=null
      tiling=NoTile
      imageSize=<16 8>
      margin=<ltrb 0.5 0 0.5 1>
      outlineSettings {
       bUseBrushTransparency=false
       roundingType=HalfHeightRadius
       width=0
       color {colorUseRule=UseColor_Specified, specifiedColor=<rgba 0 0 0 0>}
       cornerRadii=<0 0 0 0>
      }
      tintColor {colorUseRule=UseColor_Specified, specifiedColor=<rgba 1 1 1 1>}
      uVRegion {bIsValid=false, max=<0 0>, min=<0 0>}
     }
     horizontalScrolledContentPadding=<ltrb 0 0 1 0>
     leftShadowBrush {
      bIsDynamicallyLoaded=false
      drawAs=Box
      imageType=FullColor
      mirroring=NoMirror
      resourceName=../../../Engine/Content/Slate/Common/ScrollBoxShadowLeft.png
      resourceObject=null
      tiling=NoTile
      imageSize=<8 16>
      margin=<ltrb 1 0.5 0 0.5>
      outlineSettings {
       bUseBrushTransparency=false
       roundingType=HalfHeightRadius
       width=0
       color {colorUseRule=UseColor_Specified, specifiedColor=<rgba 0 0 0 0>}
       cornerRadii=<0 0 0 0>
      }
      tintColor {colorUseRule=UseColor_Specified, specifiedColor=<rgba 1 1 1 1>}
      uVRegion {bIsValid=false, max=<0 0>, min=<0 0>}
     }
     rightShadowBrush {
      bIsDynamicallyLoaded=false
      drawAs=Box
      imageType=FullColor
      mirroring=NoMirror
      resourceName=../../../Engine/Content/Slate/Common/ScrollBoxShadowRight.png
      resourceObject=null
      tiling=NoTile
      imageSize=<8 16>
      margin=<ltrb 0 0.5 1 0.5>
      outlineSettings {
       bUseBrushTransparency=false
       roundingType=HalfHeightRadius
       width=0
       color {colorUseRule=UseColor_Specified, specifiedColor=<rgba 0 0 0 0>}
       cornerRadii=<0 0 0 0>
      }
      tintColor {colorUseRule=UseColor_Specified, specifiedColor=<rgba 1 1 1 1>}
      uVRegion {bIsValid=false, max=<0 0>, min=<0 0>}
     }
     topShadowBrush {
      bIsDynamicallyLoaded=false
      drawAs=Box
      imageType=FullColor
      mirroring=NoMirror
      resourceName=../../../Engine/Content/Slate/Common/ScrollBoxShadowTop.png
      resourceObject=null
      tiling=NoTile
      imageSize=<16 8>
      margin=<ltrb 0.5 1 0.5 0>
      outlineSettings {
       bUseBrushTransparency=false
       roundingType=HalfHeightRadius
       width=0
       color {colorUseRule=UseColor_Specified, specifiedColor=<rgba 0 0 0 0>}
       cornerRadii=<0 0 0 0>
      }
      tintColor {colorUseRule=UseColor_Specified, specifiedColor=<rgba 1 1 1 1>}
      uVRegion {bIsValid=false, max=<0 0>, min=<0 0>}
     }
     verticalScrolledContentPadding=<ltrb 0 0 0 1>
    }
   }
   slot {class=VerticalBoxSlot}
  }
  HorizontalBox {widgetClassPath=&HorizontalBox, parent=ScrollBox_0, widget=HorizontalBox, slot {class=ScrollBoxSlot}}
  SeasonLabel_2 {
   widgetClassPath=&CommonTextBlock
   parent=HorizontalBox
   widget=SeasonLabel_2
   properties {
    text=Quest:
    colorAndOpacity {colorUseRule=UseColor_Specified, specifiedColor=<rgba 1 0.6514 0.0015 1>}
    font {
     bForceMonospaced=false
     bMaterialIsStencil=false
     fontMaterial=null
     fontObject=&Burbank
     letterSpacing=0
     monospacedWidth=1
     size=20
     skewAmount=0
     typefaceFontName=BurbankRegular-Black
     outlineSettings {
      bApplyOutlineToDropShadows=false
      bMiteredCorners=false
      bSeparateFillAlpha=false
      outlineMaterial=null
      outlineSize=1
      outlineColor=<rgba 0 0 0 1>
     }
    }
    shadowColorAndOpacity=<rgba 0.0052 0.0052 0.0056 0.838>
    shadowOffset=<3 3>
   }
   slot {class=HorizontalBoxSlot, properties {horizontalAlignment=HAlign_Left, padding=<ltrb 5 5 5 5>}}
  }
  QuestPicker {
   widgetClassPath=/Script/FortniteEditor.FortQuestToolPropertyView
   bIsVariable=true
   parent=HorizontalBox
   widget=QuestPicker
   properties {bShowNameWidgets=false, propertiesToShow [Quest]}
   slot {class=HorizontalBoxSlot, properties {horizontalAlignment=HAlign_Left}}
  }
  HorizontalBox_0 {widgetClassPath=&HorizontalBox, parent=ScrollBox_0, widget=HorizontalBox_0, slot {class=ScrollBoxSlot}}
  ParentQuestPropLabel {
   widgetClassPath=&CommonTextBlock
   parent=HorizontalBox_0
   widget=ParentQuestPropLabel
   properties {
    text=Parent Quest Prop Actor:
    colorAndOpacity {colorUseRule=UseColor_Specified, specifiedColor=<rgba 1 0.6514 0.0015 1>}
    font {
     bForceMonospaced=false
     bMaterialIsStencil=false
     fontMaterial=null
     fontObject=&Burbank
     letterSpacing=0
     monospacedWidth=1
     size=20
     skewAmount=0
     typefaceFontName=BurbankRegular-Black
     outlineSettings {
      bApplyOutlineToDropShadows=false
      bMiteredCorners=false
      bSeparateFillAlpha=false
      outlineMaterial=null
      outlineSize=1
      outlineColor=<rgba 0 0 0 1>
     }
    }
    shadowColorAndOpacity=<rgba 0.0052 0.0052 0.0056 0.838>
    shadowOffset=<3 3>
   }
   slot {class=HorizontalBoxSlot, properties {horizontalAlignment=HAlign_Left, padding=<ltrb 5 5 5 5>}}
  }
  ParentPropPicker {
   widgetClassPath=&FortQuestToolPropertyView
   bIsVariable=true
   parent=HorizontalBox_0
   widget=ParentPropPicker
   properties {bShowNameWidgets=false, propertiesToShow [ParentQuestPropActor]}
   slot {class=HorizontalBoxSlot, properties {horizontalAlignment=HAlign_Left}}
  }
  WrapBox_464 {widgetClassPath=&WrapBox, parent=ScrollBox_0, widget=WrapBox_464, slot {class=ScrollBoxSlot}}
  CreateChildActorsButton {
   widgetClassPath=&TA_Generic01_Button
   parent=VerticalBox_0
   widget=CreateChildActorsButton
   properties {label=Create Child Actors}
   slot {class=VerticalBoxSlot, properties {horizontalAlignment=HAlign_Right, padding=<ltrb 10 10 10 10>}}
  }
 ]
}
