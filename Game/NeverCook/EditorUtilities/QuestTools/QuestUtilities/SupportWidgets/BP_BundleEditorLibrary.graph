EditorUtilityBlueprint /Game/NeverCook/EditorUtilities/QuestTools/QuestUtilities/SupportWidgets/BP_BundleEditorLibrary
graphs [
 CreateNewBundleAsset {nodes [
   n1 = CreateNewBundleAsset!() -> (then, Quest File Picker:Object Reference, FileName:String, Folder:Directory Path Structure) {is=title K2Node_FunctionEntry}
   > n2 = GetAssetTools()
   > n3 = CreateAsset(
    self=@n2
    AssetName=@n1.FileName
    PackagePath=Break Directory Path!(DirectoryPath:"Directory Path Structure (by ref)"=@n1.Folder) -> (Path:String) {is=title K2Node_BreakStruct}
    AssetClass=/Script/FortniteGame.FortChallengeBundleItemDefinition
   )
   > Return Node!(execute, New Bundle Object:Object Reference=@n3)
  ]}
 EventGraph {links [], nodes []}
 IncrementNumberedString {nodes [
   n1 = IncrementNumberedString!() -> (then, SourceString:String) {is=title K2Node_FunctionEntry}
   > n2 = Branch(Condition=IsNumeric!(SourceString:String=Right!(SourceString:String=@n1, Count:int=1) -> (ReturnValue:String)) -> (ReturnValue:bool))
   > Set numericalCharacters(numericalCharacters:int=1) -> (Output_Get:int=0)
   n3 = Get SourceString() -> (SourceString:String)
   n4 = Get numericalCharacters() -> (numericalCharacters:int=0)
   > n5 = While Loop!(execute:Exec, Condition:bool=IsNumeric(SourceString=Right(SourceString=@n3, Count=@n4))) -> (LoopBody:Exec, Completed:Exec) {is=title K2Node_MacroInstance}
   Decrement Int(" "=@n5.Completed, Value=Get numericalCharacters() -> (numericalCharacters=0))
   n6 = Get numericalCharacters() -> (numericalCharacters=0)
   n7 = Get SourceString()
   > n8 = Increment Int(Value=String To Integer(InString=Right(SourceString=@n7, Count=@n6)))
   n9 = LeftChop(SourceString=Get SourceString(), Count=Get numericalCharacters() -> (numericalCharacters=0))
   n10 = "To String (Integer)"(InInt=@n8)
   > Return Node!(
    execute
    IncrementedString:String=SelectString(A=Append!(A=@n9, B=0, C:String=@n10) -> (ReturnValue), B=Append!(A=@n9, B=@n10) -> (ReturnValue), bPickA="integer <= integer"(A=@n8, B=9))
   )
   Increment Int(" "=@n5.LoopBody, Value=Get numericalCharacters() -> (numericalCharacters=0))
   Return Node(execute=@n2.else, IncrementedString=Append(A=Get SourceString(), B=_01))
   "To String (Integer)"(InInt=@n8)
  ]}
 Write Bundle Properties From Template {nodes [
   n1 = Write Bundle Properties From Template!() -> (then, Template To Use:Object Reference, Bundle Object To Use:Object Reference, Templated Property Names:Array of Names) {is=title K2Node_FunctionEntry}
   > Set templateObject(templateObject:Object Reference=@n1) -> (Output_Get:Object Reference)
   > n2 = For Each Loop(Array:Array of Names=@n1.Templated Property Names) -> (Array Element:Name)
   > n3 = Set currentPropertyName(currentPropertyName:Name=@n2) -> (Output_Get:Name=None)
   > n4 = GetEditorProperty!(execute:Exec, Object:Object Reference=Get Bundle Object To Use() -> (Bundle Object To Use:Object Reference), PropertyName:Name=@n3) -> (then:Exec, PropertyValue:Object Class Reference, ReturnValue:bool)
   > n5 = Cast To FortGameplayTagContainer Class!(execute:Exec, Class:Object Class Reference=@n4) -> (then:Exec, CastFailed:Exec, AsGeneric Gameplay Tag Container:Generic Gameplay Tag Container Class Reference) {is=title K2Node_ClassDynamicCast}
   > n6 = GetEditorProperty(Object=Get Bundle Object To Use(), PropertyName=Get currentPropertyName() -> (currentPropertyName:Name=None)) -> (PropertyValue:Gameplay Tag Container Structure)
   > n7 = GetEditorProperty(Object=Get templateObject() -> (templateObject:Object Reference), PropertyName=Get currentPropertyName() -> (currentPropertyName=None)) -> (PropertyValue:"Gameplay Tag Container Structure (by ref)")
   > n8 = Set gameplayTagsToWrite(gameplayTagsToWrite:Gameplay Tag Container Structure=@n6) -> (Output_Get:Gameplay Tag Container Structure)
   > AppendGameplayTagContainers!(execute, InOutTagContainer:"Gameplay Tag Container Structure (by ref)"=@n8, InTagContainer:"Gameplay Tag Container Structure (by ref)"=@n7) -> (then)
   > SetEditorProperty(
    Object=Get Bundle Object To Use()
    PropertyName=Get currentPropertyName() -> (currentPropertyName=None)
    PropertyValue:Gameplay Tag Container Structure=Get gameplayTagsToWrite() -> (gameplayTagsToWrite:Gameplay Tag Container Structure)
   )
   Execute Python Script!(
    execute=@n5.CastFailed
    PythonScript="value = template_object.get_editor_property(property_name)\r\nquest_object.set_editor_property(property_name, value)"
    template_object:Object Reference=Get templateObject()
    quest_object:Object Reference=Get Bundle Object To Use()
    property_name:Name=Get currentPropertyName() -> (currentPropertyName=None)
   ) -> (then, ReturnValue) {inputs [template_object, quest_object, property_name]}
  ]}
]
variables []
